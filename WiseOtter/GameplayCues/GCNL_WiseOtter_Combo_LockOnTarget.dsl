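; Looping gameplay cue for the WiseOtter combo lock-on, which manages a reticle effect on the target pawn.
; On loop start it caches the instigator, Niagara component and target, offsetting the effect on wildlife using capsule center of mass and the FX_Chest socket.
; An async tag-count watcher hides the reticle while the instigator has the RepulsorCannon (unibeam) tag.
; Player controller vehicle seat-change events deactivate the effect, then reactivate it on completion.
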(event EventBeginPlay
  (bind _asfort_player_controller (Utilities|Casting|CastToFortPlayerController (Game|GetPlayerController 0)))
  (Vehicle|AssignOnVehicleSeatChanged _asfort_player_controller (delegate Custom|OnVehicleSeatChanged_Event))
  (Vehicle|AssignOnVehicleSeatChangedComplete _asfort_player_controller (delegate Custom|OnVehicleSeatChangedComplete_Event)))

(event Custom|OnVehicleSeatChangedComplete_Event
  (Components|Activation|SetActive (Variables|Default|GetNS%20Looping) true true))

(event Custom|OnVehicleSeatChanged_Event
  (params (SeatIndex :type Integer))
  (Components|Activation|SetActive (Variables|Default|GetNS%20Looping) false true))

(event Collision|EventActorBeginOverlap
  (params (OtherActor :type Actor%20Object%20Reference)))

(event EventOnLoopingStart(EitherVFX-preferthisversion)
  (params (MyTarget :type Actor%20Object%20Reference) (Parameters :type Gameplay%20Cue%20Parameters%20Structure%20%28by%20ref%29) (ParticleComponents :type Array%20of%20FXSystem%20Component%20Object%20References) (AudioComponents :type Array%20of%20Audio%20Component%20Object%20References))
  (bind (_normalizedmagnitude _rawmagnitude _effectcontext _matchedtagname _originaltag _aggregatedsourcetags _aggregatedtargettags _location _normal _instigator) (Ability|GameplayCue|BreakGameplayCueParameters Parameters))
  (Utilities|FlowControl|Sequence
    (:then_0
      ; Set Instigator and Looping fx system component
      (Variables|Default|SetInstigator_0 _instigator)
      (Variables|Default|SetNS%20Looping (Utilities|Array|Get(acopy) ParticleComponents)))
    (:then_1
      ; Set Target Pawn & Adjust for Wildlife
      (bind _asplayer_pawn_athena (Utilities|Casting|CastToPlayerPawn_Athena MyTarget)
        (:then
          (Variables|Default|SetTarget%20Pawn _asplayer_pawn_athena))
        (:CastFailed
          (bind _asnpc_pawn_wildlife_parent (Utilities|Casting|CastToNPC_Pawn_Wildlife_Parent MyTarget))
          (Variables|Default|SetTarget%20Pawn _asnpc_pawn_wildlife_parent)
          (Transformation|SetRelativeLocation (Variables|Default|GetNS%20Looping) (- (Physics|GetCenterOfMass (Variables|/Script/Engine.Character|GetCapsuleComponent (Variables|Default|GetTarget%20Pawn)) "None") (Transformation|GetSocketLocation (Variables|/Script/Engine.Character|GetMesh (Variables|Default|GetTarget%20Pawn)) "FX_Chest")) false false))))
    (:then_2
      ; Show/Hide the reticle depending on if unibeam is active or not
      (bind _instigator_0 (Variables|Default|GetInstigator_0)
        (:then
          (bind _asfort_player_pawn (Utilities|Casting|CastToFortPlayerPawn _instigator_0))
          (CallFunction|ShowFX (not (GameplayTags|HasMatchingGameplayTag _asfort_player_pawn "(TagName=\"Ability.HighTower.Power.Tomato.RepulsorCannon\")")))
          (bind (_asynctaskproxy _tagcount) (Ability|Tasks|WaitGameplayTagCountChangedOnActor _asfort_player_pawn "(TagName=\"Ability.HighTower.Power.Tomato.RepulsorCannon\")")
            (:TagCountChanged
              (CallFunction|ShowFX (<= _tagcount 0)))))
        (:else)))))

(fn LockOn_Changed
  (params (LockOnTarget :type Actor%20Object%20Reference)))

(fn ShowFX
  (params (Show :type Boolean))
  (Rendering|SetVisibility (Variables|Default|GetNS%20Looping) Show false))

(fn ConstructionScript)
