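; Looping gameplay cue for the WiseOtter Unibeam beam, driving its Niagara effect and fire audio on a player pawn.
; On Niagara loop start, it traces from FX_Chest along player aim, rotates the effect, sets beam end and flags local view.
; PlayFireAudio attaches a 1P or 3P sound and places a whoosh at the laser point nearest the observer.
; Tick only runs validity checks, so the beam is not updated afterward.
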(event EventTick
  (params (DeltaSeconds :type Float%20%28single-precision%29))
  (Utilities|IsValid (stale EventOnLoopingStart(Niagara) MyTarget)
    (:Is%20Valid
      (Utilities|IsValid (Utilities|Array|Get(acopy) (stale EventOnLoopingStart(Niagara) ParticleComponents))))))

(event EventOnLoopingStart(Niagara)
  (params (MyTarget :type Actor%20Object%20Reference) (Parameters :type Gameplay%20Cue%20Parameters%20Structure%20%28by%20ref%29) (ParticleComponents :type Array%20of%20Niagara%20Particle%20System%20Component%20Object%20References) (AudioComponents :type Array%20of%20Audio%20Component%20Object%20References))
  (bind _player_pawn_athena (Variables|Default|GetPlayer%20Pawn%20Athena))
  (bind _returnvalue (Transformation|GetActorLocation _player_pawn_athena))
  (bind _player_pawn_athena_2 (Variables|Default|GetPlayer%20Pawn%20Athena))
  (bind _playeraimrotation (Variables|/Game/Athena/PlayerPawn_Athena.PlayerPawn_Athena_C|GetPlayerAimRotation _player_pawn_athena_2))
  (bind _returnvalue_2 (Math|Vector|GetForwardVector _playeraimrotation))
  (bind _range (Variables|Default|GetRange))
  (bind _returnvalue_3 (ScalableFloat|GetValueatLevel _range))
  (bind _returnvalue_4 (* _returnvalue_2 _returnvalue_3))
  (bind _returnvalue_5 (+ _returnvalue _returnvalue_4))
  (bind _output (Utilities|Array|Get(acopy) ParticleComponents))
  (bind _asplayer_pawn_athena (Utilities|Casting|CastToPlayerPawn_Athena MyTarget))
  (Variables|Default|SetPlayer%20Pawn%20Athena _asplayer_pawn_athena)
  (Utilities|FlowControl|Sequence
    (:then_0
      (bind (_outhit _returnvalue_6) (Collision|LineTraceByChannel 0 (Transformation|GetSocketLocation (Variables|/Script/Engine.Character|GetMesh _player_pawn_athena) "FX_Chest") _returnvalue_5 "TraceTypeQuery10"))
      (Transformation|SetWorldRotation _output (Variables|/Game/Athena/PlayerPawn_Athena.PlayerPawn_Athena_C|GetPlayerAimRotation (Variables|Default|GetPlayer%20Pawn%20Athena)) false false)
      (Effects|Components|ParticleSystem|SetVectorParameter _output "BeamEnd_0" (Math|Vector|SelectVector (.Location (Collision|BreakHitResult _outhit)) _returnvalue_5 _returnvalue_6)))
    (:then_1
      (CallFunction|PlayFireAudio))
    (:then_2
      (Effects|Components|ParticleSystem|SetBoolParameter _output "LocalView" (Pawn|IsLocallyControlled _asplayer_pawn_athena)))))

(event EventOnLoopingStart(Cascade)
  (params (MyTarget :type Actor%20Object%20Reference) (Parameters :type Gameplay%20Cue%20Parameters%20Structure%20%28by%20ref%29) (ParticleComponents :type Array%20of%20Cascade%20Particle%20System%20Component%20Object%20References) (AudioComponents :type Array%20of%20Audio%20Component%20Object%20References)))

(fn GetClosestPointonLasertoObserver
  (return (Math|Vector|FindClosestPointonSegment (Transformation|GetActorLocation (Game|GetPlayerPawn self)) (Transformation|GetSocketLocation (Variables|/Script/Engine.Character|GetMesh (Variables|Default|GetPlayer%20Pawn%20Athena)) "FX_Chest") (+ (Transformation|GetActorLocation) (* (Math|Vector|GetForwardVector (Variables|/Game/Athena/PlayerPawn_Athena.PlayerPawn_Athena_C|GetPlayerAimRotation (Variables|Default|GetPlayer%20Pawn%20Athena))) (ScalableFloat|GetValueatLevel (Variables|Default|GetRange)))))))

(fn OnBurst
  (params (MyTarget :type Actor%20Object%20Reference) (Parameters :type Gameplay%20Cue%20Parameters%20Structure%20%28by%20ref%29))
  (return false))

(fn PlayFireAudio
  (Utilities|FlowControl|Sequence
    (:then_0
      ; Play fire sound
      (bind _player_pawn_athena (Variables|Default|GetPlayer%20Pawn%20Athena)
        (:then
          (Audio|SpawnSoundAttached (select (Pawn|IsLocallyControlled _player_pawn_athena) (Variables|Default|GetFire%20Sound%201P) (Variables|Default|GetFire%20Sound%203P)) (Variables|/Script/Engine.Actor|GetRootComponent (Variables|Default|GetPlayer%20Pawn%20Athena))))
        (:else)))
    (:then_1
      (Utilities|FlowControl|Sequence
        (:then_0
          ; Play whoosh by sound
          (bind _player_pawn_athena_2 (Variables|Default|GetPlayer%20Pawn%20Athena)
            (:then
              (Audio|SpawnSoundatLocation _player_pawn_athena_2 (Variables|Default|GetWhoosh%20By%20Sound) (CallFunction|GetClosestPointOnLaserToObserver)))
            (:else)))))))

(fn ConstructionScript)
