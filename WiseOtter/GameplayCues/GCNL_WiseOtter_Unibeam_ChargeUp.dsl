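; Looping gameplay cue for the WiseOtter Unibeam charge-up that aims its effect and moves a fly-by sound.
; Each tick, the first particle component follows the local control rotation or, for remote players, interpolates toward replicated aim after snapping the root once.
; Non-local instances fade in the fly-by audio on loop start.
; Audio updates project the nearest in-range listener onto the root's forward axis, stopping the sound when that listener is not ahead.
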(event EventTick
  (params (DeltaSeconds :type Float%20%28single-precision%29))
  ;[1 1 1] Interpolate the fx rotation to match control rotation
  (Utilities|IsValid (stale EventOnLoopingStart(EitherVFX-preferthisversion) MyTarget)
    (:Is%20Valid
      (Utilities|IsValid (Utilities|Array|Get(acopy) (stale EventOnLoopingStart(EitherVFX-preferthisversion) ParticleComponents))
        (:Is%20Valid
          (if (Pawn|IsLocallyControlled (Utilities|Casting|CastToFortPawn (stale EventOnLoopingStart(EitherVFX-preferthisversion) MyTarget)))
            ;;[1 1 1] If this is the local client. don't used the replicated control rotation, just grab it directly.
            (Transformation|SetWorldRotation (Utilities|Array|Get(acopy) (stale EventOnLoopingStart(EitherVFX-preferthisversion) ParticleComponents)) (Pawn|GetControlRotation (Utilities|Casting|CastToFortPawn (stale EventOnLoopingStart(EitherVFX-preferthisversion) MyTarget)) :!class /Script/Engine.Pawn) false false)
            (CallFunction|UpdateAudioLocation)
            (else
              ;;[1 1 1] Interpolate the GC's rotation towards the replicated control rotation
              (bind _asplayer_pawn_athena (Utilities|Casting|CastToPlayerPawn_Athena (stale EventOnLoopingStart(EitherVFX-preferthisversion) MyTarget)))
              (if (Math|Rotator|NotEqual(Rotator) (Variables|/Game/Athena/PlayerPawn_Athena.PlayerPawn_Athena_C|GetPlayerAimRotation _asplayer_pawn_athena) (Variables|Default|GetSavedRotation) 0.0001)
                (Utilities|DoOnceThenDo false
                  (:DoOnce
                    (Transformation|SetWorldRotation (Variables|Transformation|GetRootComponent) (Variables|/Game/Athena/PlayerPawn_Athena.PlayerPawn_Athena_C|GetPlayerAimRotation _asplayer_pawn_athena) false false))
                  (:ThenDo
                    (Transformation|SetWorldRotation (Utilities|Array|Get(acopy) (stale EventOnLoopingStart(EitherVFX-preferthisversion) ParticleComponents)) (Math|Interpolation|RInterpTo (Transformation|GetWorldRotation (Utilities|Array|Get(acopy) (stale EventOnLoopingStart(EitherVFX-preferthisversion) ParticleComponents))) (Variables|/Game/Athena/PlayerPawn_Athena.PlayerPawn_Athena_C|GetPlayerAimRotation _asplayer_pawn_athena) DeltaSeconds 2e+01) false false)))
                (CallFunction|UpdateAudioLocation)
                (Variables|Default|SetSavedRotation (Variables|/Game/Athena/PlayerPawn_Athena.PlayerPawn_Athena_C|GetPlayerAimRotation _asplayer_pawn_athena))))))))))

(event EventOnLoopingStart(EitherVFX-preferthisversion)
  (params (MyTarget :type Actor%20Object%20Reference) (Parameters :type Gameplay%20Cue%20Parameters%20Structure%20%28by%20ref%29) (ParticleComponents :type Array%20of%20FXSystem%20Component%20Object%20References) (AudioComponents :type Array%20of%20Audio%20Component%20Object%20References))
  ;[1 1 1] Interpolate the fx rotation to match control rotation
  (Utilities|FlowControl|Sequence
    (:then_0
      (Utilities|FlowControl|Sequence
        (:then_0
          _)
        (:then_1
          ;;[1 1 1] Interpolate the GC's rotation towards the replicated control rotation
          (bind _asplayer_pawn_athena (Utilities|Casting|CastToPlayerPawn_Athena MyTarget))
          (if (Math|Rotator|NotEqual(Rotator) (Variables|/Game/Athena/PlayerPawn_Athena.PlayerPawn_Athena_C|GetPlayerAimRotation _asplayer_pawn_athena) (Variables|Default|GetSavedRotation) 0.0001)
            (Utilities|DoOnceThenDo false
              (:DoOnce
                (Transformation|SetWorldRotation (Variables|Transformation|GetRootComponent) (Variables|/Game/Athena/PlayerPawn_Athena.PlayerPawn_Athena_C|GetPlayerAimRotation _asplayer_pawn_athena) false false))
              (:ThenDo
                (Transformation|SetWorldRotation (Utilities|Array|Get(acopy) ParticleComponents) (Math|Interpolation|RInterpTo (Transformation|GetWorldRotation (Utilities|Array|Get(acopy) ParticleComponents)) (Variables|/Game/Athena/PlayerPawn_Athena.PlayerPawn_Athena_C|GetPlayerAimRotation _asplayer_pawn_athena) (stale EventTick DeltaSeconds) 2e+01) false false)))
            (CallFunction|UpdateAudioLocation)
            (Variables|Default|SetSavedRotation (Variables|/Game/Athena/PlayerPawn_Athena.PlayerPawn_Athena_C|GetPlayerAimRotation _asplayer_pawn_athena))))))
    (:then_1
      (if (not (Pawn|IsLocallyControlled (Utilities|Casting|CastToFortPawn MyTarget)))
        (Audio|Components|Audio|FadeIn (Variables|Default|GetHightower_Tomato_Secondary_Charge_FlyBy) 0.5 1.0 0.0 "Linear" :!class /Script/Engine.AudioComponent)))))

(event Custom|Update%20Audio%20Location
  (bind _rootcomponent (Variables|Transformation|GetRootComponent))
  (bind _returnvalue (Transformation|GetForwardVector _rootcomponent))
  (bind _rootcomponent_2 (Variables|Transformation|GetRootComponent))
  (bind _returnvalue_2 (Transformation|GetWorldLocation _rootcomponent_2))
  (bind _hightower_tomato_secondary_charge_flyby (Variables|Default|GetHightower_Tomato_Secondary_Charge_FlyBy))
  ;[1 1 1] Move audio component
  (bind (_listenerposition _returnvalue_3) (Audio|GetClosestListenerLocation (Variables|Transformation|GetRootComponent) _returnvalue_2 (ScalableFloat|GetValueatLevel (Variables|Default|GetMaxRange)) true))
  (if _returnvalue_3
    (if (Audio|Components|Audio|IsPlaying (Variables|Default|GetHightower_Tomato_Secondary_Charge_FlyBy))
      (if (> (Math|Vector|DotProduct _returnvalue (- _listenerposition _returnvalue_2)) 0.0)
        ;[1 1 1] Interpolate the fx rotation to match control rotation
        (Transformation|SetWorldLocation _hightower_tomato_secondary_charge_flyby (+ _returnvalue_2 (Math|Vector|ProjectVectorontoVector (- _listenerposition _returnvalue_2) _returnvalue)) false false)
        (else
          (Audio|Components|Audio|Stop _hightower_tomato_secondary_charge_flyby))))))

(fn OnExecute
  (params (MyTarget :type Actor%20Object%20Reference) (Parameters :type Gameplay%20Cue%20Parameters%20Structure%20%28by%20ref%29))
  (return false))

(fn ConstructionScript)
