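; Gameplay cue for the WiseOtter Unibeam structure hit, handling impact effects and gunshot sound indicators.
; On burst, it shows indicators at the target and, given a valid instigator, the hit location, then aligns each particle component to the impact normal.
; Per particle, non-Character instigators get either a water splash with sound or a fading decal, depending on a water-body cast.
; On execute, it spawns a radial force field if physics fields are enabled.
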(fn OnBurst(EitherVFX-preferthisversion)
  (params (MyTarget :type Actor%20Object%20Reference) (Parameters :type Gameplay%20Cue%20Parameters%20Structure%20%28by%20ref%29) (ParticleComponents :type Array%20of%20FXSystem%20Component%20Object%20References) (AudioComponents :type Array%20of%20Audio%20Component%20Object%20References) (BurstCameraShakeInstance :type Camera%20Shake%20Base%20Object%20Reference) (BurstDecalInstance :type Decal%20Actor%20Object%20Reference))
  (bind (_normalizedmagnitude _rawmagnitude _effectcontext _matchedtagname _originaltag _aggregatedsourcetags _aggregatedtargettags _location _normal _instigator) (Ability|GameplayCue|BreakGameplayCueParameters Parameters))
  (CallFunction|SpawnSoundIndicators MyTarget Parameters)
  (for _array_element ParticleComponents
    (Transformation|SetWorldLocation _array_element _location false false)
    (Transformation|SetWorldRotation _array_element (Math|Conversions|RotationFromXVector _normal) false false)
    (Utilities|Casting|CastToCharacter _instigator
      (:CastFailed
        (Utilities|Casting|CastToFortWaterBodyActor _instigator
          (:then
            (Effects|Components|ParticleSystem|SpawnFXAtLocation _instigator "/Game/Effects/Fort_Effects/GameplayPulginsAssets/Water/NS_Gameplay_LandImpactWater.NS_Gameplay_LandImpactWater" _location)
            (Audio|PlaySoundatLocation 0 "/Game/Athena/Sounds/Weapons/Explosions/Expl_Debris_Water_Cue.Expl_Debris_Water_Cue" _location))
          (:CastFailed
            (bind _returnvalue (Rendering|Decal|SpawnDecalatLocation 0 "/WiseOtter/Assets/FX/Unibeam/MD_WiseOtter_Decal.MD_WiseOtter_Decal" "200.000000,300.000000,300.000000" _location (Math|Conversions|RotationFromXVector _normal) 3.0))
            (Rendering|Components|Decal|SetFadeOut _returnvalue 3.0 2.0 true))))))
  (return))

(fn OnExecute
  (params (MyTarget :type Actor%20Object%20Reference) (Parameters_NormalizedMagnitude :type Float%20%28single-precision%29) (Parameters_RawMagnitude :type Float%20%28single-precision%29) (Parameters_EffectContext :type Gameplay%20Effect%20Context%20Handle%20Structure) (Parameters_MatchedTagName :type Gameplay%20Tag%20Structure) (Parameters_OriginalTag :type Gameplay%20Tag%20Structure) (Parameters_AggregatedSourceTags :type Gameplay%20Tag%20Container%20Structure) (Parameters_AggregatedTargetTags :type Gameplay%20Tag%20Container%20Structure) (Parameters_Location :type Vector :default "0, 0, 0") (Parameters_Normal :type Vector :default "0, 0, 0") (Parameters_Instigator :type Actor%20Object%20Reference) (Parameters_EffectCauser :type Actor%20Object%20Reference) (Parameters_SourceObject :type Object%20Reference) (Parameters_PhysicalMaterial :type Physical%20Material%20Object%20Reference) (Parameters_GameplayEffectLevel :type Integer) (Parameters_AbilityLevel :type Integer) (Parameters_TargetAttachComponent :type Scene%20Component%20Object%20Reference) (Parameters_bReplicateLocationWhenUsingMinimalRepProxy :type Boolean))
  (bind _returnvalue (Development|GetConsoleVariableIntValue "r.PhysicsField.EnableField"))
  (if (== _returnvalue 1)
    (Game|SpawnActorfromClass 0 "/Game/Effects/Niagara/Blueprints/PhsyicsFields/FN_RadialForce.FN_RadialForce_C" (Math|Transform|MakeTransform Parameters_Location) "Undefined" "MultiplyWithRoot" 0 0.7 0.75 1e+03 1.5 0.0 0))
  (return (|CallParentFunction (Variables|GetMyTarget) (Variables|GetParameters))))

(fn SpawnSoundIndicators
  (params (Target :type Actor%20Object%20Reference) (Params :type Gameplay%20Cue%20Parameters%20Structure))
  (bind (_normalizedmagnitude _rawmagnitude _effectcontext _matchedtagname _originaltag _aggregatedsourcetags _aggregatedtargettags _location _normal _instigator) (Ability|GameplayCue|BreakGameplayCueParameters Params))
  (bind _returnvalue (Team|GetActorTeam Target))
  (bind _array (Utilities|Array|MakeArray "Hostile" "Neutral"))
  (Fortnite|ShowSoundIndicatorLocalforTeam :WorldContextObject Target :TrackedActor Target :Location (Transformation|GetActorLocation Target) :MaxAudibleDistance 1e+04 :IndicatorType "Gunshot" :Team _returnvalue :Affiliations _array :Instigator Target :Tint "(R=0.000000,G=0.000000,B=0.000000,A=0.000000)" :AttenuationMaxRangeMultiplier 1.0 :AttenuationIntensityMultiplier 1.0)
  (Utilities|IsValid _instigator
    (:Is%20Valid
      (Fortnite|ShowSoundIndicatorLocalforTeam :WorldContextObject Target :TrackedActor Target :Location _location :MaxAudibleDistance 1e+04 :IndicatorType "Gunshot" :Team _returnvalue :Affiliations _array :Instigator Target :Tint "(R=0.000000,G=0.000000,B=0.000000,A=0.000000)" :AttenuationMaxRangeMultiplier 1.0 :AttenuationIntensityMultiplier 1.0))))
